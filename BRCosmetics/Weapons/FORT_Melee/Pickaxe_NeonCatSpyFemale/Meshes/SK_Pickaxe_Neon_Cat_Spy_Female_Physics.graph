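# Physics asset for the Neon Cat Spy Female pickaxe's skeletal mesh, providing bodies for a dangling bone chain on the weapon.
# Three capsule bodies on dyn_dangle_01 to 03 are jointed in series from a tiny anchor capsule on dyn_dangle_01_MSTR.
# The root joint locks twist and one swing axis, so the chain pivots there on a single axis, while each later joint carries its own swing and twist settings.
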PhysicsAsset /BRCosmetics/Weapons/FORT_Melee/Pickaxe_NeonCatSpyFemale/Meshes/SK_Pickaxe_Neon_Cat_Spy_Female_Physics
bodies {
 dyn_dangle_01 [{shapeType=Capsule, length=4.141, radius=2.5, center=<0.8177 0 0.0001>, rotation=<rot -90 -0.0148 0.0148>}]
 dyn_dangle_01_MSTR [{shapeType=Capsule, length=0.01, radius=0.01}]
 dyn_dangle_02 [{shapeType=Capsule, length=4.141, radius=2.5, center=<0.939 0 0>, rotation=<rot -90 -0.0066 0.0066>}]
 dyn_dangle_03 [{shapeType=Capsule, length=4.141, radius=2.5, center=<6.1364 0 0>, rotation=<rot -90 179.9802 -179.9802>}]
}
constraints [
 {bone1Name=dyn_dangle_02, bone2Name=dyn_dangle_01, swing1LimitDegrees=90, swing2LimitDegrees=90, twistLimitDegrees=25}
 {bone1Name=dyn_dangle_03, bone2Name=dyn_dangle_02, swing1LimitDegrees=90, swing2LimitDegrees=15, twistLimitDegrees=45}
 {
  bone1Name=dyn_dangle_01
  bone2Name=dyn_dangle_01_MSTR
  swing1LimitDegrees=0
  swing1Motion=Locked
  swing2LimitDegrees=90
  twistLimitDegrees=0
  twistMotion=Locked
 }
]
